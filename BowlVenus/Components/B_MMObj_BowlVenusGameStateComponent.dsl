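; Game state component for the BowlVenus capture-flag objective, formerly named radio tower, driving its data layers and activation phases.
; On BeginPlay the server activates data layers, with a deterministic datatable-order option.
; After objective activation, a timer estimates replicated progress from remaining safe-zone time; a hotfixed storm phase plus delay completes it.
; It also handles closest-flag marking, debug cheats and captured-flag analytics.
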(event EventBeginPlay
  ;[0.01502 0.175 0.1823] BeginPlay: Set up actor registry and enable the radio tower datalayers.
  (|CallParentFunction)
  (Networking|SwitchHasAuthority
    (:Authority
      (Variables|Default|SetCaptureFlagPhaseState "NewEnumerator0")
      (Default|CallOnCaptureFlagPhaseAdvance)
      (bind _gamestate (Class|FortAsyncActionCurrentPlaylistReady|PlaylistDataReadyAsync)
        (:OnReady
          (Fortnite|FortShippingLog self "BeginPlay: Invoked.")
          (General|LogDataLayers)
          (Activation|ActivateDataLayers)
          (Utilities|IsValid _gamestate
            (:Is%20Valid
              (bind _returnvalue (GameplayEventRouter|GetEventRouter "/Script/FortniteGame.FortGameState" _gamestate))
              (bind (_asynctaskproxy _payload) (GameplayEventRouter|StartListeningToStatefulEvent _returnvalue "/Script/FortniteGame.FortGamePhaseUpdatedEvent")
                (:then
                  (Variables|Default|SetGamePhaseUpdatedAsyncTask _asynctaskproxy))
                (:OnEventReceived
                  _)
                (:OnSavedEventStateExists
                  _))
              (Analytics|HandleGamePhaseChanged (Utilities|Struct|BreakFortGamePhaseUpdatedEvent _payload)))))))))

(event EventActivateObjective
  ;[0.04187 0.1979 0.08506] ActivateObjective: When the objective is activated, activate the capture flag
  (|CallParentFunction)
  (Networking|SwitchHasAuthority
    (:Authority
      (Utilities|FlowControl|DoOnce false
        (:Completed
          (Fortnite|FortShippingLog self "ActivateObjective: Invoked")
          ;;[0.4635 0.197 0.08937] Capture Flag should enter the Activating state.
          (Variables|Default|SetCaptureFlagPhaseState "NewEnumerator6")
          (Default|CallOnCaptureFlagPhaseAdvance)
          ;;[0.03732 0.1247 0.1771] Setup timers and variables to start updating the activation progress.
          (Variables|Transient|Activation|SetObjectiveActivateServerWorldTime (Fortnite|Sync|Time|GetServerWorldTimeSeconds))
          (bind _returnvalue (Utilities|Time|SetTimerbyEvent (EventDispatchers|CreateEvent :!event UpdateCaptureFlagActivationProgress) (Math|Float|Max(Float) (ScalableFloat|GetValueatLevel (Variables|Hotfix|GetHotfix_UpdateCaptureFlagActivationProgressInterval)) 1.0) true))
          (Variables|Transient|Activation|SetUpdateCaptureFlagActivationProgressTimerHandle _returnvalue)
          (Activation|UpdateCaptureFlagActivationProgress))))))

(event Custom|CaptureFlagActivationComplete
  ;[0.01285 0.1171 0.2604] RadioTowerActivationComplete: Once the activation progress is complete, let the Capture Flag know that it should enter the Active state.
  (Utilities|SwitchHasAuthority
    (:Authority
      (Utilities|FlowControl|DoOnce false
        (:Completed
          (Variables|Transient|Activation|SetCaptureFlagActivationProgress 1.0)
          (if (ScalableFloat|GetValueasBool (Variables|Hotfix|GetHotfix_MarkCaptureFlagOnObjectiveActivate))
            (Variables|Transient|Activation|SetbRepNotifyMarkCaptureFlag true)
            (else
              _))
          (Utilities|Time|ClearandInvalidateTimerbyHandle (Variables|Transient|Activation|GetUpdateCaptureFlagActivationProgressTimerHandle))
          (Utilities|Time|ClearandInvalidateTimerbyHandle (Variables|Transient|GetActivateCompleteDelayTimerHandle))
          (Variables|Default|SetCaptureFlagPhaseState "NewEnumerator2")
          (Default|CallOnCaptureFlagPhaseAdvance))))))

(event EventEndPlay
  (params (EndPlayReason :type EEndPlayReason%20Enum :default "Destroyed"))
  ;[0.1667 0.08823 0.02458] EndPlay: Unbind delegates.
  (Utilities|FlowControl|Sequence
    (:then_0
      (Utilities|SwitchHasAuthority
        (:Authority
          (Utilities|FlowControl|Sequence
            (:then_1
              (if (Editor|InEditororPIE)
                (Analytics|SubmitAnalytics)
                (else
                  _))
              (bind _gamephaseupdatedasynctask (Variables|Default|GetGamePhaseUpdatedAsyncTask)
                (:then
                  (GameplayEventRouter|StopListeningtoStatefulEvent _gamephaseupdatedasynctask))
                (:else)))))))
    (:then_1
      (|CallParentFunction EndPlayReason))))

(event Custom|DebugCaptureFlagActivateObjective
  ;[0 0 0] Cheat ke * DebugRadioTowerActivateObjective
  ;
  ; Activate the radio tower objective.
  (Networking|SwitchHasAuthority
    (:Authority
      (CallFunction|ActivateObjective)
      (CallFunction|CaptureFlagActivationComplete))))

(event Custom|DebugCaptureFlagActivateDatalayerOnly
  (params (Input :type String))
  ;[0 0 0] Cheat ke * DebugRadioTowerActivateDatalayerOnly (names)
  ;
  ; Activates the desired radio tower datalayers and disable the others.
  (Networking|SwitchHasAuthority
    (:Authority
      (CallFunction|DebugCaptureFlagDeactivateDatalayer "all")
      (CallFunction|DebugCaptureFlagActivateDatalayer Input))))

(event Custom|DebugCaptureFlagActivateDatalayer
  (params (Input :type String))
  ;[0 0 0] Cheat ke * DebugRadioTowerActivateDatalayer (names)
  ;
  ; Activates the desired radio tower datalayers.
  (Networking|SwitchHasAuthority
    (:Authority
      (bind _radiotowerspawninfodatatable (Variables|Settings|GetRadioTowerSpawnInfoDatatable)
        (:then
          (bind _outrownames (DataTable|GetDataTableRowNames _radiotowerspawninfodatatable))
          (for _array_element _outrownames
            (bind _returnvalue (Utilities|GetDataTableRow _radiotowerspawninfodatatable _array_element))
            (if (ScalableFloat|GetValueasBool (Utilities|Struct|BreakMMObjPOISpawnInfoStruct _returnvalue))
              (Utilities|IsValid (.DatalayerAsset_19_35CF832546D6BB81AB591C9AEB62F17D (Utilities|Struct|BreakMMObjPOISpawnInfoStruct _returnvalue))
                (:Is%20Valid
                  (if (or (Utilities|String|Equal,CaseInsensitive(String) (Utilities|String|ToLower Input) "all") (Utilities|Array|ContainsItem (Utilities|String|ParseIntoArray (Utilities|String|ToLower Input) ",") (Utilities|String|ToLower (Utilities|String|ToString(Name) _array_element))))
                    (bind _returnvalue_2 (WorldPartition|GetDataLayerManager))
                    (Utilities|IsValid _returnvalue_2
                      (:Is%20Valid
                        (DataLayers|SetDataLayerRuntimeState _returnvalue_2 (.DatalayerAsset_19_35CF832546D6BB81AB591C9AEB62F17D (Utilities|Struct|BreakMMObjPOISpawnInfoStruct _returnvalue)) "Activated" false :!class /Script/Engine.DataLayerManager)))))))))
        (:else)))))

(event Custom|DebugCaptureFlagDeactivateDatalayer
  (params (Input :type String))
  ;[0 0 0] Cheat ke * DebugRadioTowerDeactivateDatalayer (names)
  ;
  ; Deactivates the desired radio tower datalayers.
  (Networking|SwitchHasAuthority
    (:Authority
      (bind _radiotowerspawninfodatatable (Variables|Settings|GetRadioTowerSpawnInfoDatatable)
        (:then
          (bind _outrownames (DataTable|GetDataTableRowNames _radiotowerspawninfodatatable))
          (for _array_element _outrownames
            (bind _returnvalue (Utilities|GetDataTableRow _radiotowerspawninfodatatable _array_element))
            (if (ScalableFloat|GetValueasBool (Utilities|Struct|BreakMMObjPOISpawnInfoStruct _returnvalue))
              (Utilities|IsValid (.DatalayerAsset_19_35CF832546D6BB81AB591C9AEB62F17D (Utilities|Struct|BreakMMObjPOISpawnInfoStruct _returnvalue))
                (:Is%20Valid
                  (if (or (Utilities|String|Equal,CaseInsensitive(String) (Utilities|String|ToLower Input) "all") (Utilities|Array|ContainsItem (Utilities|String|ParseIntoArray (Utilities|String|ToLower Input) ",") (Utilities|String|ToLower (Utilities|String|ToString(Name) _array_element))))
                    (bind _returnvalue_2 (WorldPartition|GetDataLayerManager))
                    (Utilities|IsValid _returnvalue_2
                      (:Is%20Valid
                        (DataLayers|SetDataLayerRuntimeState _returnvalue_2 (.DatalayerAsset_19_35CF832546D6BB81AB591C9AEB62F17D (Utilities|Struct|BreakMMObjPOISpawnInfoStruct _returnvalue)) "Unloaded" false :!class /Script/Engine.DataLayerManager)))))))))
        (:else)))))

(fn ActivateDataLayers
  ;[0.0614 0.1927 0.05798] ActivateDataLayers: Activate the radio tower datalayers in the world.
  (Networking|SwitchHasAuthority
    (:Authority
      (if (not (Variables|Transient|Activation|GetbDataLayersActivated))
        (Variables|Transient|Activation|SetbDataLayersActivated true)
        (Fortnite|FortShippingLog self "ActivateDataLayers: Invoked successfully.")
        (if (Class|FortDeterminismConfig|IsUsingRadioTowerDeterminism)
          (Fortnite|FortShippingLog self "ActivateDataLayers: Deterministic active.")
          (Activation|ActivateDeterministicDatalayers)
          (else
            (Activation|ActivateNextDataLayer)))
        (else
          ;;[1 0 0] Warning: Datalayers are already activated.
          (Fortnite|FortShippingLog self "ActivateDataLayers: DataLayers are already activated." true))))))

(fn ActivateDeterministicDatalayers
  ;[0.007986 0.1136 0.1562] ActivateDeterministicDatalayer: Activate the first found valid datalayers in the datatable.
  (Networking|SwitchHasAuthority
    (:Authority
      (bind _returnvalue (WorldPartition|GetDataLayerManager))
      (Utilities|IsValid _returnvalue
        (:Is%20Valid
          (bind _radiotowerspawninfodatatable (Variables|Settings|GetRadioTowerSpawnInfoDatatable)
            (:then
              (bind _outrownames (DataTable|GetDataTableRowNames _radiotowerspawninfodatatable))
              (for _array_element _outrownames
                (Utilities|FlowControl|Sequence
                  (:then_0
                    (if (>= (Utilities|Array|Length (Variables|Transient|Activation|GetActivatedDataLayers)) (ScalableFloat|GetValueasInteger (Variables|Hotfix|GetHotfix_NumOfCaptureFlag)))
                      (return)))
                  (:then_1
                    (bind _returnvalue_2 (Utilities|GetDataTableRow _radiotowerspawninfodatatable _array_element))
                    (if (ScalableFloat|GetValueasBool (Utilities|Struct|BreakMMObjPOISpawnInfoStruct _returnvalue_2))
                      (Utilities|IsValid (.DatalayerAsset_19_35CF832546D6BB81AB591C9AEB62F17D (Utilities|Struct|BreakMMObjPOISpawnInfoStruct _returnvalue_2))
                        (:Is%20Valid
                          (bind _returnvalue_3 (DataLayers|SetDataLayerRuntimeState _returnvalue (.DatalayerAsset_19_35CF832546D6BB81AB591C9AEB62F17D (Utilities|Struct|BreakMMObjPOISpawnInfoStruct _returnvalue_2)) "Activated" false :!class /Script/Engine.DataLayerManager))
                          (if _returnvalue_3
                            (Utilities|Array|Add (Variables|Transient|Activation|GetActivatedDataLayers) _array_element)))))))))
            (:else)))))))

(fn ActivateNextDataLayer)

(fn BuildDataRegistryResolverScope
  (return 0 0 false))

(fn CompareGamePhaseState
  (params (A_SafeZonePhaseIndex :type Integer) (A_SafeZoneState :type EFortSafeZoneState%20Enum) (B_SafeZonePhaseIndex :type Integer) (B_SafeZoneState :type EFortSafeZoneState%20Enum))
  ;[0.04344 0.1719 0.02753] Compare phase index and state to figure out if A happens after B.
  (Math|Integer|CompareInt A_SafeZonePhaseIndex B_SafeZonePhaseIndex
    (:>
      (return true))
    (:==
      (return (> (Math|Conversions|ToInteger(Byte) A_SafeZoneState) (Math|Conversions|ToInteger(Byte) B_SafeZoneState))))
    (:<
      (return false))))

(fn GetCacheContextOverride
  (return 0))

(fn GetCaptureFlag
  (bind _outactors (Actor|GetAllActorsOfClass "/POIBanner/Actors/BGA_POIBanner_Banner.BGA_POIBanner_Banner_C"))
  (return _outactors))

(fn GetCaptureFlagActivationProgress
  (return (Variables|Transient|Activation|GetCaptureFlagActivationProgress)))

(fn GetCaptureFlagDatalayerLocations_Activated
  ;[0.02087 0.151 0.1345] Returns the locations where the datalayer has already been activated.
  ; Called from B_MMObj_EQC_ActivatedRadioTowerDatalayerLocations.
  (Utilities|FlowControl|Sequence
    (:then_0
      (for _array_element (Variables|Transient|Activation|GetActivatedDataLayers)
        (bind _radiotowerspawninfodatatable (Variables|Settings|GetRadioTowerSpawnInfoDatatable)
          (:then
            (bind _returnvalue (Utilities|GetDataTableRow _radiotowerspawninfodatatable _array_element))
            (if (ScalableFloat|GetValueasBool (Utilities|Struct|BreakMMObjPOISpawnInfoStruct _returnvalue))
              (Utilities|Array|Add (LocalVariables|GetRetLocations) (.SpawnLocation_15_5033779E420594BA62C2D5A783E10B93 (Utilities|Struct|BreakMMObjPOISpawnInfoStruct _returnvalue)))))
          (:else))))
    (:then_1
      (return (LocalVariables|GetRetLocations)))))

(fn GetCaptureFlagDatalayerLocations_Pending
  ;[0.01686 0.09047 0.1302] Return the locations where radio tower datalayers can still be activated.
  ; Called from B_MMObj_EQS_RadioTowerDatalayerLocationGenerator.
  (Utilities|FlowControl|Sequence
    (:then_0
      (Utilities|SwitchHasAuthority
        (:Authority
          (bind _radiotowerspawninfodatatable (Variables|Settings|GetRadioTowerSpawnInfoDatatable)
            (:then
              (bind _outrownames (DataTable|GetDataTableRowNames _radiotowerspawninfodatatable))
              (for _array_element _outrownames
                (if (not (Utilities|Array|ContainsItem (Variables|Transient|Activation|GetActivatedDataLayers) _array_element))
                  (bind _returnvalue (Utilities|GetDataTableRow _radiotowerspawninfodatatable _array_element))
                  (if (ScalableFloat|GetValueasBool (Utilities|Struct|BreakMMObjPOISpawnInfoStruct _returnvalue))
                    (Utilities|Array|Add (LocalVariables|GetRetLocations) (.SpawnLocation_15_5033779E420594BA62C2D5A783E10B93 (Utilities|Struct|BreakMMObjPOISpawnInfoStruct _returnvalue)))))))
            (:else)))))
    (:then_1
      (return (LocalVariables|GetRetLocations)))))

(fn GetCaptureFlagState
  (return (Variables|Default|GetCaptureFlagPhaseState)))

(fn HandleCapturedFlagState
  (params (CaptureFlag :type BGA%20POIBanner%20Banner%20Object%20Reference) (NewState :type MMObjBowlVenusState%20Enum :default "NewEnumerator5"))
  ;[0.05924 0.09937 0.1302] HandleRadioTowerStateChanged: Track all the radio towers that have been captured for analytics.
  (Networking|SwitchHasAuthority
    (:Authority
      (Utilities|IsValid CaptureFlag
        (:Is%20Valid
          (if (Utilities|Enum|Equal(Enum) NewState "NewEnumerator7")
            (Utilities|Array|AddUnique (Variables|Analytics|GetAnalytics_CapturedRadioTowers) (Utilities|GetDisplayName CaptureFlag))))))))

(fn HandleActivateCompleteDelay
  ;[0.009761 0.1379 0.1771] HandleActivateCompleteDelay: Called when the activate storm phase index and state are reached and the specified delay period has passed.
  (Utilities|SwitchHasAuthority
    (:Authority
      (CallFunction|CaptureFlagActivationComplete))))

(fn HandleDatalayerSelectionQueryComplete
  (params (QueryInstance :type EQS%20Query%20Instance%20Object%20Reference) (QueryStatus :type EEnvQueryStatus%20Enum :default "Processing")))

(fn HandleGamePhaseChanged
  (params (GamePhase :type EAthenaGamePhase%20Enum))
  ;[0.0164 0.07935 0.1198] HandleGamePhaseChanged: Once we're in EndGame, submit the analytics.
  (Networking|SwitchHasAuthority
    (:Authority
      (if (Utilities|Enum|Equal(Enum) GamePhase "EndGame")
        (Analytics|SubmitAnalytics)
        (bind _gamephaseupdatedasynctask (Variables|Default|GetGamePhaseUpdatedAsyncTask)
          (:then
            (GameplayEventRouter|StopListeningtoStatefulEvent _gamephaseupdatedasynctask))
          (:else))))))

(fn HandleSafeZoneUpdated
  (params (PhaseIndex :type Integer) (SafeZoneState :type EFortSafeZoneState%20Enum))
  (bind _hotfix_activatecompletedelay (Variables|Hotfix|GetHotfix_ActivateCompleteDelay))
  (bind _returnvalue (ScalableFloat|GetValueatLevel _hotfix_activatecompletedelay))
  ;[0.006416 0.05939 0.0625] HandleSafeZoneUpdated: Overridden from parent to get a notification on whenever the safe zone phase or state is updated.
  (|CallParentFunction PhaseIndex SafeZoneState)
  (Utilities|SwitchHasAuthority
    (:Authority
      (if (and (== (ScalableFloat|GetValueasInteger (Variables|Hotfix|GetHotfix_ActivateCompleteSafeZonePhaseIndex)) PhaseIndex) (Utilities|Enum|Equal(Enum) (Utilities|Enum|BytetoEnumEFortSafeZoneState (Math|Conversions|ToByte(Integer) (ScalableFloat|GetValueasInteger (Variables|Hotfix|GetHotfix_ActivateCompleteSafeZoneState)))) SafeZoneState))
        (Variables|Transient|SetbActivateCompleteSafeZoneStateAndPhaseReached true)
        ;;[0.1094 0.02868 0.1086] IF any additional delay is specified, run the timer for that duration.
        (if (> _returnvalue 0)
          (bind _returnvalue_2 (Utilities|Time|SetTimerbyEvent (EventDispatchers|CreateEvent :!event HandleActivateCompleteDelay) _returnvalue))
          (Variables|Transient|SetActivateCompleteDelayTimerHandle _returnvalue_2)
          (else
            (Activation|HandleActivateCompleteDelay)))))))

(fn IncrementGamePhaseState
  (params (SafeZonePhaseIndex :type Integer) (SafeZoneState :type EFortSafeZoneState%20Enum))
  ;[0.02715 0.08376 0.1562] Increment the safe zone phase/state.
  (switch Utilities|FlowControl|Switch|SwitchonEFortSafeZoneState SafeZoneState
    (:None
      _)
    (:Starting
      _)
    (:Holding
      (return SafeZonePhaseIndex "Shrinking"))
    (:Shrinking
      (return (+ SafeZonePhaseIndex 1) "Holding")))
  (return SafeZonePhaseIndex SafeZoneState))

(fn LogDataLayers
  ;[0.1042 0.04513 0.1667] LogDataLayers: Based on hotfix values, print out all the enable and disabled datalayers for debugging purposes.
  (Networking|SwitchHasAuthority
    (:Authority
      (bind _radiotowerspawninfodatatable (Variables|Settings|GetRadioTowerSpawnInfoDatatable)
        (:then
          (bind _outrownames (DataTable|GetDataTableRowNames _radiotowerspawninfodatatable))
          (for _array_element _outrownames
            (bind _returnvalue (Utilities|GetDataTableRow _radiotowerspawninfodatatable _array_element))
            (Utilities|IsValid (.DatalayerAsset_19_35CF832546D6BB81AB591C9AEB62F17D (Utilities|Struct|BreakMMObjPOISpawnInfoStruct _returnvalue))
              (:Is%20Valid
                (if (ScalableFloat|GetValueasBool (Utilities|Struct|BreakMMObjPOISpawnInfoStruct _returnvalue))
                  (Fortnite|FortShippingLog self (Utilities|String|Append "LogDataLayers: Enabled: " (Utilities|String|ToString(Name) _array_element)))
                  (else
                    (Fortnite|FortShippingLog self (Utilities|String|Append "LogDataLayers: Disabled: " (Utilities|String|ToString(Name) _array_element))))))
              (:Is%20Not%20Valid
                (Fortnite|FortShippingLog self (Utilities|String|Append "LogDataLayers: DataLayerAsset for Row " (Utilities|String|ToString(Name) _array_element) " is NULL.") true)))))
        (:else
          ;;[1 0.02106 0] Warning: RadioTowerSpawnInfoDatatable is null.
          (Fortnite|FortShippingLog self "LogDataLayers: RadioTowerSpawnInfoDatatable is null.." true))))))

(fn MarkCaptureFlaglForAllPlayers
  ;[0.09896 0.01585 0.07957] Loop through all the players and mark the closest radio tower terminal for each player.
  (Networking|CanExecuteCosmeticEvents
    (:True
      (LocalVariables|SetMarkOnObjectiveActivateDist2D (ScalableFloat|GetValueatLevel (Variables|Hotfix|GetHotfix_MarkCaptureFlagOnObjectiveActivateDistance)))
      (bind _captureflag (General|GetCaptureFlag))
      (if (Utilities|Array|IsNotEmpty _captureflag)
        (for _array_element (Fortnite|GetLocalFortPlayerControllers true)
          (Utilities|IsValid _array_element
            (:Is%20Valid
              (Utilities|IsValid (Pawn|GetViewTarget _array_element)
                (:Is%20Valid
                  (bind _asfort_player_state_athena (Utilities|Casting|CastToFortPlayerStateAthena (Variables|/Script/Engine.Controller|GetPlayerState _array_element)))
                  (if (<= (Variables|/Script/FortniteGame.FortPlayerStateAthena|GetPlace _asfort_player_state_athena) 0)
                    (Utilities|IsValid (Actor|GetComponentbyClass _array_element "/Script/FortniteGame.AthenaMarkerComponent")
                      (:Is%20Valid
                        (bind _asbga_poibanner_banner (Utilities|Casting|CastToBGA_POIBanner_Banner (Fortnite|GetClosestActorfromArray (Pawn|GetViewTarget _array_element) _captureflag)))
                        (if (not (<= (LocalVariables|GetMarkOnObjectiveActivateDist2D) 0))
                          (if (<= (Math|Vector|Distance2DSquared(Vector) (Transformation|GetActorLocation _asbga_poibanner_banner) (Transformation|GetActorLocation (Pawn|GetViewTarget _array_element))) (Math|Float|Square (LocalVariables|GetMarkOnObjectiveActivateDist2D)))
                            _))
                        (MarkableActor|SetCanBeMarked _asbga_poibanner_banner true) ; Ensure the actor can be marked in case its state variable hasn't replicated yet.
                        (Fortnite|RegisterSoundIndicatorPlacementContext _array_element _asbga_poibanner_banner "0, 0, 0" false)))))))))))))

(fn NotifyCaptureFlagObjectiveComplete
  (params (PlayerPawn :type Fort%20Player%20Pawn%20Object%20Reference))
  ;[0.01101 0.07779 0.09375] NotifyRadioTowerObjectiveComplete: Used to mark this "objective" as complete for the input player.
  ; Called directly from BGA_MMObj_BeetCalfTerminal.
  (Networking|SwitchHasAuthority
    (:Authority
      (Utilities|IsValid PlayerPawn
        (:Is%20Valid
          (bind _asfort_player_state_athena (Utilities|Casting|CastToFortPlayerStateAthena (Variables|/Script/Engine.Pawn|GetPlayerState PlayerPawn)))
          (CallFunction|SetObjectiveCompletedStat (Class|FortPlayerStateAthena|GetSquadId _asfort_player_state_athena)))))))

(fn OnRep_bRepNotifyMarkCaptureFlag
  ;[0.1198 0.003449 0.1034] Mark the capture flag for all players.
  (Networking|CanExecuteCosmeticEvents
    (:True
      (if (Variables|Transient|Activation|GetbRepNotifyMarkCaptureFlag)
        (General|MarkCaptureFlaglforAllPlayers)))))

(fn OnRep_CaptureFlagActivationProgress
  ;[0.2079 0.05571 0.2083] Broadcast the latest RadioTowerActivationProgress.
  (Default|CallOnCaptureFlagActivationProgressUpdated (Variables|Transient|Activation|GetCaptureFlagActivationProgress)))

(fn SubmitAnalytics
  (bind _analytics_capturedradiotowers (Variables|Analytics|GetAnalytics_CapturedRadioTowers))
  (bind _returnvalue (Utilities|Array|Length _analytics_capturedradiotowers))
  (bind _analytics_capturedradiotowers_2 (Variables|Analytics|GetAnalytics_CapturedRadioTowers))
  (bind _returnvalue_2 (Utilities|String|JoinStringArray _analytics_capturedradiotowers_2 " ; "))
  ;[0.03754 0.005307 0.04167] SubmitAnalytics: Fire analytics if it hasn't been done yet.
  (Utilities|SwitchHasAuthority
    (:Authority
      (if (and (not (Replay|GetIsPlayingReplay)) (not (Variables|Analytics|GetbAnalyticsSubmitted)))
        (Variables|Analytics|SetbAnalyticsSubmitted true)
        (Utilities|FlowControl|Sequence
          (:then_0
            ;;[0.02283 0.151 0.08822] NumOfCapturedRadioTowers
            (bind _returnvalue_3 (Analytics|ConstructAnalyticsAttribute :AttributeName "NumOfCapturedRadioTowers" :IntValue _returnvalue :floatValue -1.0 :BoolValue false))
            (Utilities|Array|Add (LocalVariables|GetAnalyticsEventAttributes) _returnvalue_3)
            (Fortnite|FortShippingLog self (Utilities|String|Append "NumOfCapturedRadioTowers: " (Utilities|String|ToString(Integer) _returnvalue)))
            ;;[0.02424 0.08952 0.1823] CapturedRadioTowers
            (bind _returnvalue_4 (Analytics|ConstructAnalyticsAttribute "CapturedRadioTowers" _returnvalue_2))
            (Utilities|Array|Add (LocalVariables|GetAnalyticsEventAttributes) _returnvalue_4)
            (Fortnite|FortShippingLog self (Utilities|String|Append "CapturedRadioTowers: " _returnvalue_2)))
          (:then_1
            (Analytics|FireAnalyticsEventBlueprintArray "Core.MMObj.RadioTower" (LocalVariables|GetAnalyticsEventAttributes))))))))

(fn UpdateCaptureFlagActivationProgress
  (bind _returnvalue (BattleRoyaleGamePhaseLogic|GetBattleRoyaleGamePhaseLogicComponent))
  ;[0.1797 0.03639 0.1875] UpdateRadioTowerActivationProgress: Update the radio tower activation progress over time.
  (Networking|SwitchHasAuthority
    (:Authority
      (Utilities|FlowControl|Sequence
        (:then_0
          ;;[0.01776 0.278 0.3021] Cache how long it has been since the activation started.
          (LocalVariables|SetTimeSinceActivationStarted (- (Fortnite|Sync|Time|GetServerWorldTimeSeconds) (Variables|Transient|Activation|GetObjectiveActivateServerWorldTime)))
          ;;[0.02785 0.2612 0.3073] Cache the next SafeZonePhaseIndex and SafeZoneState.
          (bind _safezoneindicator (Variables|Default|GetSafeZoneIndicator)
            (:then
              (bind (_outsafezonephaseindex _outsafezonestate) (Helper|IncrementGamePhaseState (SafeZone|GetCurrentPhaseIndex _safezoneindicator) (SafeZone|GetSafeZoneState _safezoneindicator)))
              (LocalVariables|SetSafeZonePhaseIndex _outsafezonephaseindex)
              (LocalVariables|SetSafeZoneState _outsafezonestate)
              ;;[0.02615 0.009237 0.1406] Loop through all the future states to find the amount of time (in seconds) it is going to take to get there.
              (while (Helper|CompareGamePhaseState (ScalableFloat|GetValueasInteger (Variables|Hotfix|GetHotfix_ActivateCompleteSafeZonePhaseIndex)) (Utilities|Enum|BytetoEnumEFortSafeZoneState (Math|Conversions|ToByte(Integer) (ScalableFloat|GetValueasInteger (Variables|Hotfix|GetHotfix_ActivateCompleteSafeZoneState)))) (LocalVariables|GetSafeZonePhaseIndex) (LocalVariables|GetSafeZoneState))
                ;;;[1 0.5885 0.1497] Tally up the time left in the various future states.
                (bind _outxy (DataTable|EvaluateCurveTableRow "/Game/Athena/Balance/DataTables/AthenaGameData.AthenaGameData" (select (Utilities|Enum|Equal(Enum) (LocalVariables|GetSafeZoneState) "Holding") "Default.SafeZone.WaitTime" "Default.SafeZone.ShrinkTime") (Math|Conversions|ToFloat(Integer) (LocalVariables|GetSafeZonePhaseIndex)))
                  (:RowFound
                    (LocalVariables|SetFutureSafeZonesTimeLeft (+ (LocalVariables|GetFutureSafeZonesTimeLeft) _outxy))
                    ;;;[0.03069 0.1927 0.04119] Increment the variables for the next loop.
                    (bind (_outsafezonephaseindex_2 _outsafezonestate_2) (Helper|IncrementGamePhaseState (LocalVariables|GetSafeZonePhaseIndex) (LocalVariables|GetSafeZoneState)))
                    (LocalVariables|SetSafeZonePhaseIndex _outsafezonephaseindex_2)
                    (LocalVariables|SetSafeZoneState _outsafezonestate_2)))))
            (:else
              ;;[1 0 0] Warning: SafeZoneIndicator is null.
              (Fortnite|FortShippingLog self "UpdateRadioTowerActivationProgress: SafeZoneIndicator is NULL." true))))
        (:then_1
          ;;[0.03419 0.08364 0.1771] Update the RadioTowerActivationProgress in a normalized ratio.
          (Utilities|IsValid _returnvalue
            (:Is%20Valid
              (bind _output_get (Variables|Transient|Activation|SetCaptureFlagActivationProgress (Math|Float|SafeDivide (LocalVariables|GetTimeSinceActivationStarted) (Utilities|Operators|Add (select (Variables|Transient|GetbActivateCompleteSafeZoneStateAndPhaseReached) (Utilities|Time|GetTimerRemainingTimebyHandle (Variables|Transient|GetActivateCompleteDelayTimerHandle)) (+ (BattleRoyaleGamePhaseLogic|GetGamePhaseStepTimeRemaining _returnvalue) (ScalableFloat|GetValueatLevel (Variables|Hotfix|GetHotfix_ActivateCompleteDelay)))) (LocalVariables|GetTimeSinceActivationStarted) (LocalVariables|GetFutureSafeZonesTimeLeft)))))
              (Development|PrintString (Utilities|String|Append "RadioTowerActivationProgress: " (Utilities|String|ToString(Float) _output_get)) true false "(R=0.000000,G=0.660000,B=1.000000,A=1.000000)" 2.0 "RadioTower"))))))))
