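; Player controller component that connects a player to POI banner capture points, driving capture UI, sounds and gameplay tags.
; Discovery volume overlaps on the possessed pawn bind banner progress, raise and capture-state events.
; OwningClient events forward these to the POI UI controller and play team-relative sounds.
; Gameplay effects mark capturing and capture state on the pawn, and a per-POI cooldown from a console variable throttles captured toasts.
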(event Custom|ListenForLocationTags
  (bind _ownerpawn (Variables|Default|GetOwnerPawn))
  (Ability|Tasks|WaitGameplayTagAddToActor _ownerpawn (Variables|Default|GetPOIParentTag)
    (:then
      (Ability|Tasks|WaitGameplayTagRemoveFromActor _ownerpawn (Variables|Default|GetPOIParentTag)
        (:Removed
          (CallFunction|OnLocationExited))))
    (:Added
      (bind (_outtag _tagfound) (CallFunction|GetLocationTagonPawnOwner))
      (if _tagfound
        (CallFunction|OnLocationEntered _outtag)))))

(event EventBeginPlay
  ;[1 1 1] Initialization
  (CallFunction|InitializeOwnerReferences)
  (CallFunction|InitializePOIClaimManagerReference)
  (Utilities|IsLocallyControlled))

(event Custom|OnProgressUpdated
  (params (Team :type Integer) (CurrentProgress :type Float%20%28double-precision%29))
  ;[1 1 1] Progress Updates
  (CallFunction|OwningClientUpdateProgress Team CurrentProgress))

(event EventTick
  (params (DeltaSeconds :type Float%20%28single-precision%29)))

(event Custom|OnPossessedPawnChanged
  (params (OldPawn :type Pawn%20Object%20Reference) (NewPawn :type Pawn%20Object%20Reference)))

(event Custom|OwningClientUpdateProgress
  (params (TeamIndex :type Integer) (CurrentProgress :type Float%20%28double-precision%29))
  (bind _returnvalue (CallFunction|IsMyTeam TeamIndex))
  (bind _retcomponent (CallFunction|GetorAddUIController))
  ;[1 1 1] Progress Updates
  (Utilities|IsValid _retcomponent
    (:Is%20Valid
      (Default|CallOnUpdateProgress _retcomponent _returnvalue CurrentProgress))
    (:Is%20Not%20Valid
      _))
  ;;[1 1 1] Play progress sound only if the new progress received was a different affiliation than the last
  (if (!= (Variables|Default|GetLastTeam) TeamIndex)
    (Variables|Default|SetLastTeam TeamIndex)
    (Audio|PlaySound2D (select _returnvalue (Variables|Default|GetSound_FriendlyProgress) (Variables|Default|GetSound_EnemyProgress)))))

(event Custom|OnPlayerPawnPossessed
  (params (PossessedPawn :type Pawn%20Object%20Reference))
  ;[1 1 1] Initialization
  (bind _asfort_player_pawn (Utilities|Casting|CastToFortPlayerPawn PossessedPawn))
  (Variables|Default|SetOwnerPawn _asfort_player_pawn)
  (CallFunction|BindOverlapListener))

(event Custom|OnOwnerPawnBeginOverlap
  (params (OverlappedActor :type Actor%20Object%20Reference) (OtherActor :type Actor%20Object%20Reference))
  ;[1 1 1] Overlap Listeners
  (bind _asb_poiclaim_discovery_volume (Utilities|Casting|CastToB_POIClaim_DiscoveryVolume OtherActor))
  (CallFunction|OnEnteredDiscoveryVolume (Variables|/POIBanner/Actors/B_POIClaim_DiscoveryVolume.B_POIClaim_DiscoveryVolume_C|GetBanner _asb_poiclaim_discovery_volume)))

(event Custom|OnOwnerPawnEndOverlap
  (params (OverlappedActor :type Actor%20Object%20Reference) (OtherActor :type Actor%20Object%20Reference))
  ;[1 1 1] Overlap Listeners
  (Utilities|Casting|CastToB_POIClaim_DiscoveryVolume OtherActor)
  (CallFunction|OnExitedDiscoveryVolume))

(event Custom|OnEnteredDiscoveryVolume
  (params (Banner :type BGA%20POIBanner%20Banner%20Object%20Reference) (LocationTag :type Gameplay%20Tag%20Structure))
  ;[1 1 1] POI Volume Overlap Callbacks
  ;;[1 1 1] On Entered
  (bind _output_get (Variables|Default|SetCurrentBanner Banner))
  (CallFunction|BindBannerListeners _output_get)
  (Variables|Default|SetCurrentLocationTag LocationTag)
  (Variables|Default|SetbIsInDiscoveryVolume true)
  (CallFunction|OwningClientonEnteredDiscoveryVolume _output_get))

(event Custom|OwningClient_OnEnteredDiscoveryVolume
  (params (Banner :type BGA%20POIBanner%20Banner%20Object%20Reference)))

(event Custom|OnExitedDiscoveryVolume
  ;[1 1 1] POI Volume Overlap Callbacks
  ;;[1 1 1] On Exited
  (Variables|Default|SetbIsInDiscoveryVolume false)
  (CallFunction|OwningClientonExitedDiscoveryVolume)
  (CallFunction|CleanUpBannerListeners))

(event Custom|OwningClient_OnExitedDiscoveryVolume
  (bind _retcomponent (CallFunction|GetorAddUIController))
  ;[1 1 1] POI Volume Overlap Callbacks
  ;;[1 1 1] On Exited
  (Utilities|IsValid _retcomponent
    (:Is%20Valid
      ;;;[1 1 1] Can handle anything else that needs to be hidden here
      (Default|CallForceUpdateVisibility _retcomponent "Collapsed" 0))))

(event Custom|OnCurrentCapturePointStateChanged
  (params (CapturePoint :type BGA%20POIBanner%20Capture%20Point%20Object%20Reference) (NewState :type EPOIBannerState%20Enum :default "NewEnumerator0") (OldState :type EPOIBannerState%20Enum :default "NewEnumerator0") (LastCapturingTeam :type Integer))
  ;[1 1 1] Capture Point State Change Handlers
  (CallFunction|OwningClientonCurrentCapturePointStateChanged CapturePoint NewState OldState LastCapturingTeam)
  (Tags|UpdateCaptureStateTag NewState))

(event Custom|OnPOICaptured
  (params (Team :type Integer) (Winners :type Array%20of%20Fort%20Player%20Pawn%20Object%20References) (Banner :type BGA%20POIBanner%20Banner%20Object%20Reference))
  ;[1 1 1] On Captured
  (if (Default|GetShowPoiToastonCapture Banner)
    (CallFunction|ShowCapturedPOIToast Team Winners Banner)
    (else
      _))
  (CallFunction|OwningClientonPOIClaimed (CallFunction|IsMyTeam Team))
  (Tags|UpdateCapturingTag false)
  (Tags|RemoveCaptureStateTag))

(event Custom|OnCapturePointEntered
  (params (CapturePoint :type BGA%20POIBanner%20Capture%20Point%20Object%20Reference) (Banner :type BGA%20POIBanner%20Banner%20Object%20Reference))
  ;[1 1 1] Capture Point Overlap Callbacks
  ;;[1 1 1] Callback when we enter a capture point. Passing in the associated banner here as well in case we somehow enter the capture point without overlapping a POI volume or they both happen simultaneously
  (Variables|Default|SetCurrentCapturePoint CapturePoint)
  ;;;[1 1 1] Set banner ref if our old one is different from the new one (these could already be the same if set from the POI volume first)
  (if (!= (Variables|Default|GetCurrentBanner) Banner)
    (bind _output_get (Variables|Default|SetCurrentBanner Banner))
    (CallFunction|BindBannerListeners _output_get)
    (else
      _))
  (CallFunction|OwningClientonCapturePointEntered CapturePoint Banner)
  (Tags|UpdateCapturingTag true))

(event Custom|OwningClient_OnCurrentCapturePointStateChanged
  (params (CapturePoint :type BGA%20POIBanner%20Capture%20Point%20Object%20Reference) (NewState :type EPOIBannerState%20Enum :default "NewEnumerator0") (OldState :type EPOIBannerState%20Enum :default "NewEnumerator0") (LastCapturingTeam :type Integer))
  (bind _retcomponent (CallFunction|GetorAddUIController))
  ;[1 1 1] Capture Point State Change Handlers
  (Utilities|FlowControl|Sequence
    (:then_0
      ;;[1 1 1] Check if the point is contested
      (bind _iscontested (CallFunction|DetermineContested NewState LastCapturingTeam))
      (if _iscontested
        (Audio|PlaySound2D (Variables|Default|GetSound_Contested))
        (else
          _))
      (Utilities|IsValid _retcomponent
        (:Is%20Valid
          (Default|CallOnPOIContested _retcomponent _iscontested))))))

(event Custom|ShowCapturedPOIToast
  (params (Team :type Integer) (Winners :type Array%20of%20Fort%20Player%20Pawn%20Object%20References) (Banner :type BGA%20POIBanner%20Banner%20Object%20Reference))
  (bind _poitag (Variables|/POIBanner/Actors/BGA_POIBanner_Banner.BGA_POIBanner_Banner_C|GetPOITag Banner))
  ;[1 1 1] On Captured
  ;;[0 0 0] Toast cooldown: If POI is not coolingdown, then register current time in the map and display the toast.
  (if (not (CallFunction|IsPOICoolingDown _poitag))
    (Utilities|Map|Add (Variables|Default|GetPOIToastCooldown) _poitag (Utilities|Time|GetGameTimeinSeconds))
    (CallFunction|UpdateTempPlayerAvatars Winners)
    (CallFunction|OwningClientonShowCapturedPOIToast (CallFunction|IsMyTeam Team) _poitag (Variables|Default|GetTempPlayerAvatars))))

(event Custom|OwningClient_OnCapturePointEntered
  (params (CapturePoint :type BGA%20POIBanner%20Capture%20Point%20Object%20Reference) (Banner :type BGA%20POIBanner%20Banner%20Object%20Reference))
  (bind _retcomponent (CallFunction|GetorAddUIController))
  ;[1 1 1] Capture Point Overlap Callbacks
  (Utilities|IsValid _retcomponent
    (:Is%20Valid
      (Default|CallOnPOIEntered _retcomponent CapturePoint)
      (Default|CallOnUpdateVisibility _retcomponent "HitTestInvisible" (Variables|/POIBanner/Actors/CapturePoint/BGA_POIBanner_CapturePoint.BGA_POIBanner_CapturePoint_C|GetCurrentCaptureProgress CapturePoint)))))

(event Custom|OwningClient_OnShowCapturedPOIToast
  (params (Won :type Boolean) (POITag :type Gameplay%20Tag%20Structure) (PlayerAvatars :type Array%20of%20Texture%202D%20Soft%20Object%20References))
  (bind _retcomponent (CallFunction|GetorAddUIController))
  ;[1 1 1] On Captured
  (Audio|PlaySound2D (select Won (Variables|Default|GetSound_Won) (Variables|Default|GetSound_Lost)))
  (Utilities|IsValid _retcomponent
    (:Is%20Valid
      (Default|CallOnShowPOIToast _retcomponent Won POITag PlayerAvatars))))

(event Custom|OnCapturePointExited
  (params (CapturePoint :type BGA%20POIBanner%20Capture%20Point%20Object%20Reference) (Banner :type BGA%20POIBanner%20Banner%20Object%20Reference))
  ;[1 1 1] Capture Point Overlap Callbacks
  (CallFunction|OwningClientonCapturePointExited CapturePoint Banner)
  (Tags|UpdateCapturingTag false)
  (Tags|RemoveCaptureStateTag))

(event Custom|OwningClient_OnPOIClaimed
  (params (Won :type Boolean))
  (bind _retcomponent (CallFunction|GetorAddUIController))
  ;[1 1 1] On Captured
  (Utilities|IsValid _retcomponent
    (:Is%20Valid
      (Default|CallOnPOIClaimed _retcomponent Won))))

(event Custom|OwningClient_OnCapturePointExited
  (params (CapturePoint :type BGA%20POIBanner%20Capture%20Point%20Object%20Reference) (Banner :type BGA%20POIBanner%20Banner%20Object%20Reference))
  (bind _retcomponent (CallFunction|GetorAddUIController))
  ;[1 1 1] Capture Point Overlap Callbacks
  (if (and (Utilities|Enum|NotEqual(Enum) (Variables|/POIBanner/Actors/CapturePoint/BGA_POIBanner_CapturePoint.BGA_POIBanner_CapturePoint_C|GetState CapturePoint) "NewEnumerator3") (Variables|Default|GetbIsInDiscoveryVolume))
    ;;[1 1 1] Updates visibility (primarily of screen-anchored widget, but others too
    (Utilities|IsValid _retcomponent
      (:Is%20Valid
        (Default|CallOnUpdateVisibility _retcomponent "Collapsed" (Variables|/POIBanner/Actors/CapturePoint/BGA_POIBanner_CapturePoint.BGA_POIBanner_CapturePoint_C|GetCurrentCaptureProgress CapturePoint))))))

(fn BindBannerListeners
  (params (NewBanner :type BGA%20POIBanner%20Banner%20Object%20Reference))
  (Utilities|IsValid NewBanner
    (:Is%20Valid
      (Default|BindEventtoOnProgressUpdated NewBanner (EventDispatchers|CreateEvent :!event OnProgressUpdated))
      (Default|BindEventtoOnBannerRaised NewBanner (EventDispatchers|CreateEvent :!event OnPOICaptured))
      (bind _capturepoint (Class|BGAPOIBannerBanner|GetFirstProgressSourceasCapturePoint NewBanner))
      (bind _output_get (Variables|Default|SetCurrentCapturePoint _capturepoint))
      (EventDispatchers|BindEventtoOnCaptureStateChanged _output_get (EventDispatchers|CreateEvent :!event OnCurrentCapturePointStateChanged)))))

(fn BindOverlapListener
  (bind _ownerpawn (Variables|Default|GetOwnerPawn))
  (Collision|BindEventtoOnActorBeginOverlap _ownerpawn (EventDispatchers|CreateEvent :!event OnOwnerPawnBeginOverlap))
  (Collision|BindEventtoOnActorEndOverlap _ownerpawn (EventDispatchers|CreateEvent :!event OnOwnerPawnEndOverlap))
  (bind _output_get (LocalVariables|SetOverlappingActors (Collision|GetOverlappingActors _ownerpawn "/POIBanner/Actors/B_POIClaim_DiscoveryVolume.B_POIClaim_DiscoveryVolume_C" :!class /Script/Engine.Actor)))
  (for _array_element _output_get
    (CallFunction|OnEnteredDiscoveryVolume (Variables|/POIBanner/Actors/B_POIClaim_DiscoveryVolume.B_POIClaim_DiscoveryVolume_C|GetBanner _array_element))))

(fn CheckPOIVolumeBeginOverlap
  (params (OtherActor :type Actor%20Object%20Reference))
  (bind _manager (Variables|Default|GetManager))
  (bind _asfort_poi_volume (Utilities|Casting|CastToFortPoiVolume OtherActor))
  (bind _outtag (CallFunction|GetPOIVolumeLocationTag _asfort_poi_volume))
  (if (GameplayTags|IsGameplayTagValid _outtag)
    (Utilities|IsValid _manager
      (:Is%20Valid
        (bind (_value _returnvalue) (Class|BPOIClaimManager|GetBannerforLocation _manager _outtag))
        (if _returnvalue
          (CallFunction|OnEnteredDiscoveryVolume _value _outtag)
          (else
            _))))
    (else
      _))
  (Development|PrintString "Could not find banner for location"))

(fn CheckPOIVolumeEndOverlap
  (params (OtherActor :type Actor%20Object%20Reference))
  (Utilities|Casting|CastToFortPoiVolume OtherActor)
  (CallFunction|OnExitedDiscoveryVolume))

(fn CleanUpBannerListeners
  (bind _currentbanner (Variables|Default|GetCurrentBanner))
  (bind _currentcapturepoint (Variables|Default|GetCurrentCapturePoint))
  (Utilities|FlowControl|Sequence
    (:then_0
      (Utilities|IsValid _currentbanner
        (:Is%20Valid
          (Default|UnbindEventfromOnProgressUpdated _currentbanner (EventDispatchers|CreateEvent :!event OnProgressUpdated))
          (Default|UnbindEventfromOnBannerRaised _currentbanner (EventDispatchers|CreateEvent :!event OnPOICaptured)))))
    (:then_1
      (Utilities|IsValid _currentcapturepoint
        (:Is%20Valid
          (EventDispatchers|UnbindEventfromOnCaptureStateChanged _currentcapturepoint (EventDispatchers|CreateEvent :!event OnCurrentCapturePointStateChanged)))))))

(fn DetermineContested
  (params (State :type EPOIBannerState%20Enum :default "NewEnumerator0") (Team :type Integer))
  (return (Utilities|Enum|Equal(Enum) State "NewEnumerator2")))

(fn GetOrAddUIController
  (Utilities|IsValid (Variables|Default|GetUIController)
    (:Is%20Valid
      _)
    (:Is%20Not%20Valid
      (Variables|Default|SetUIController (Actor|GetComponentbyClass (Variables|Default|GetOwnerController) "/POIBanner/UI/BP_POIWin_UIController.BP_POIWin_UIController_C"))))
  (return (Variables|Default|GetUIController)))

(fn GetPOIVolumeLocationTag
  (params (POIVolume :type Fort%20Poi%20Volume%20Object%20Reference))
  (for _array_element (GameplayTags|BreakGameplayTagContainer (Variables|/Script/FortniteGame.FortPoiVolume|GetLocationTags POIVolume))
    (if (GameplayTags|MatchesTag _array_element "(TagName=\"Athena.Location.POI\")")
      (return _array_element)))
  (return 0))

(fn GetCurrentBannerProgress)

(fn GetLocationTagOnPawnOwner
  (for _array_element (GameplayTags|BreakGameplayTagContainer (GameplayTags|GetOwnedGameplayTags (Variables|Default|GetOwnerPawn) :!class /Script/GameplayTags.GameplayTagAssetInterface))
    (if (GameplayTags|MatchesTag _array_element (Variables|Default|GetPOIParentTag))
      (return _array_element true)))
  (return 0 false))

(fn GetMyTeam
  (return (Teams|TeamtoIndex (Team|GetActorTeam (Variables|Default|GetOwnerController)))))

(fn GetOwnerAbilitySystemComponent
  (bind _ownerpawn (Variables|Default|GetOwnerPawn)
    (:then
      (return (Ability|GetFortAbilitySystemComponent _ownerpawn)))
    (:else)))

(fn InitializeOwnerReferences
  (bind _asfort_player_controller (Utilities|Casting|CastToFortPlayerController (Components|GetOwner)))
  (bind _output_get (Variables|Default|SetOwnerController _asfort_player_controller))
  (Game|Player|BindEventtoOnPlayerPawnPossessed _output_get (EventDispatchers|CreateEvent :!event OnPlayerPawnPossessed))
  (Utilities|IsValid (Pawn|GetControlledPawn _output_get)
    (:Is%20Valid
      (bind _asfort_player_pawn (Utilities|Casting|CastToFortPlayerPawn (Pawn|GetControlledPawn _output_get)))
      (Variables|Default|SetOwnerPawn _asfort_player_pawn)
      (CallFunction|BindOverlapListener)
      (Development|PrintString "Found owner pawn"))))

(fn InitializePOIClaimManagerReference
  (bind _outactors (Actor|GetAllActorsOfClass "/POIBanner/Actors/B_POIClaim_Manager.B_POIClaim_Manager_C"))
  (if (Utilities|Array|IsValidIndex _outactors)
    (Variables|Default|SetManager (Utilities|Array|Get(acopy) _outactors))))

(fn InitializeProgressOnEntered
  (params (Banner :type BGA%20POIBanner%20Banner%20Object%20Reference))
  (bind (_hasprogress _value) (Class|BGAPOIBannerBanner|GetCurrentTeamProgress Banner))
  (if _hasprogress
    (CallFunction|OnProgressUpdated (Variables|/POIBanner/Actors/BGA_POIBanner_Banner.BGA_POIBanner_Banner_C|GetCurrentTeam Banner) _value)
    (else
      (CallFunction|OnProgressUpdated 0))))

(fn IsMyTeam
  (params (TeamIndex :type Integer))
  (return (== TeamIndex (CallFunction|GetMyTeam))))

(fn IsPOICoolingDown
  (params (POITag :type Gameplay%20Tag%20Structure))
  (bind _poitoastcooldown (Variables|Default|GetPOIToastCooldown))
  (bind (_value _returnvalue) (Utilities|Map|Find _poitoastcooldown POITag))
  (if _returnvalue
    (bind _returnvalue_2 (Development|GetConsoleVariableFloatValue "Fort.HUD.ToastCooldownDuration"))
    (return (< (- (Utilities|Time|GetGameTimeinSeconds) _value) _returnvalue_2))
    (else
      (return false))))

(fn OnLocationEntered
  (params (LocationTag :type Gameplay%20Tag%20Structure))
  (bind _manager (Variables|Default|GetManager))
  (Utilities|IsValid _manager
    (:Is%20Valid
      (bind (_value _returnvalue) (Class|BPOIClaimManager|GetBannerforLocation _manager LocationTag))
      (if _returnvalue
        (CallFunction|OnEnteredDiscoveryVolume _value LocationTag)
        (else
          _)))
    (:Is%20Not%20Valid
      _))
  (Development|PrintString "Could not find banner for location"))

(fn OnLocationExited
  (CallFunction|OnExitedDiscoveryVolume))

(fn RemoveCaptureStateTag
  (bind _returnvalue (CallFunction|GetOwnerAbilitySystemComponent))
  (Utilities|IsValid _returnvalue
    (:Is%20Valid
      (GameplayEffects|RemoveActiveEffectswithTags _returnvalue (Variables|Default|GetCaptureStateTags)))))

(fn UpdateCaptureStateTag
  (params (State :type EPOIBannerState%20Enum%20%28by%20ref%29 :default "NewEnumerator0"))
  (bind _capturestategameplayeffectmap (Variables|Default|GetCaptureStateGameplayEffectMap))
  (bind (_value _returnvalue) (Utilities|Map|Find _capturestategameplayeffectmap State))
  (bind _returnvalue_2 (CallFunction|GetOwnerAbilitySystemComponent))
  ;[1 1 1] Remove existing capture state tag so we don't double up
  (Tags|RemoveCaptureStateTag)
  (if _returnvalue
    (Utilities|IsValid _returnvalue_2
      (:Is%20Valid
        (GameplayEffects|ApplyGameplayEffectToSelf :self _returnvalue_2 :GameplayEffectClass _value :Level 0.0)))))

(fn UpdateCapturingTag
  (params (IsInCapturePoint :type Boolean))
  (bind _returnvalue (CallFunction|GetOwnerAbilitySystemComponent))
  (bind _returnvalue_2 (CallFunction|GetOwnerAbilitySystemComponent))
  (Networking|SwitchHasAuthority
    (:Authority
      (if IsInCapturePoint
        ;[1 1 1] Don't push the tag if the capture point is already captured (banner raised)
        (if (not (Variables|/POIBanner/Actors/BGA_POIBanner_Banner.BGA_POIBanner_Banner_C|GetbIsRaised (Variables|Default|GetCurrentBanner)))
          (Utilities|IsValid _returnvalue_2
            (:Is%20Valid
              (GameplayEffects|ApplyGameplayEffectToSelf :self _returnvalue_2 :GameplayEffectClass (Variables|Default|GetGE_CapturingTag) :Level 0.0))))
        (else
          ;[1 1 1] Remove the granted tag if we are no longer within the point or it is otherwise invalid
          (Utilities|IsValid _returnvalue
            (:Is%20Valid
              (GameplayEffects|RemoveActiveEffectswithGrantedTags _returnvalue (Variables|Default|GetCapturingGrantedTags)))))))))

(fn UpdateTempPlayerAvatars
  (params (InputPin :type Array%20of%20Fort%20Player%20Pawn%20Object%20References))
  (Utilities|Array|Clear (Variables|Default|GetTempPlayerAvatars))
  (for _array_element InputPin
    (bind _asfort_player_state_athena (Utilities|Casting|CastToFortPlayerStateAthena (Variables|/Script/Engine.Pawn|GetPlayerState _array_element)))
    (Utilities|Array|Add (Variables|Default|GetTempPlayerAvatars) (Class|FortPlayerStateZone|GetSocialAvatarBrush _asfort_player_state_athena true))))
